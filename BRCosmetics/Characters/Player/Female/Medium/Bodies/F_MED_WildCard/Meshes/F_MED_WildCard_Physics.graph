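# Physics asset for the F_MED_WildCard female character body, covering two extra cloth pieces around the hips.
# Pelvis and thigh capsules form the body shapes, with thighs and tiny cloth master anchors constrained freely in rotation to the pelvis.
# Each cloth bone hangs from its master with a hinge-like constraint: wide on one swing axis, narrow on the other, zero twist.
# The left piece uses three angled boxes, the right a single box.
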PhysicsAsset /BRCosmetics/Characters/Player/Female/Medium/Bodies/F_MED_WildCard/Meshes/F_MED_WildCard_Physics
bodies {
 extra_cloth_MSTR_l [{shapeType=Capsule, length=0.505, radius=0.505, shapeName=extra_cloth_MSTR_l_capsule, rotation=<rot 0 -90 0>}]
 extra_cloth_MSTR_r [{shapeType=Capsule, length=0.505, radius=0.505, shapeName=extra_cloth_MSTR_r_capsule, rotation=<rot 0 -90 0>}]
 extra_cloth_l [
  {
   shapeType=Box
   extentX=3.5
   extentY=0.5
   extentZ=7.5
   shapeName=extra_cloth_r_box
   center=<7.484 -0.2998 -1.071>
   rotation=<rot 6.0106 0.0007 -178.7895>
  }
  {
   shapeType=Box
   extentX=3.5
   extentY=0.5
   extentZ=4.5
   shapeName=extra_cloth_l_box
   center=<6.5774 -0.8059 4.7736>
   rotation=<rot 12.3978 -3.6665 -16.6539>
  }
  {
   shapeType=Box
   extentX=3.5
   extentY=0.5
   extentZ=2
   shapeName=extra_cloth_l_box
   center=<5.7899 -2.3039 7.5879>
   rotation=<rot 12.0168 -5.3576 -41.1459>
  }
 ]
 extra_cloth_r [{
   shapeType=Box
   extentX=3.5
   extentY=0.5
   extentZ=10.5
   shapeName=extra_cloth_r_box
   center=<6.3479 0.0385 -1.2457>
   rotation=<rot -6.7438 0.942 -2.5685>
  }]
 pelvis [{shapeType=Capsule, length=13.2506, radius=14.2316, shapeName=pelvis_capsule, center=<0.8469 -6.837 -0.1165>, rotation=<rot 0 -90 90>}]
 thigh_l [{shapeType=Capsule, length=15.3742, radius=8.1698, shapeName=thigh_l_capsule, center=<-12.0212 0.1343 0>, rotation=<rot 0 -96.9044 -90>}]
 thigh_r [{shapeType=Capsule, length=15.3742, radius=8.1698, shapeName=thigh_l_capsule, center=<12.0212 -0.1343 0>, rotation=<rot 0 -96.9044 90>}]
}
constraints [
 {
  bone1Name=thigh_l
  bone2Name=pelvis
  swing1LimitDegrees=45
  swing1Motion=Free
  swing2LimitDegrees=45
  swing2Motion=Free
  twistLimitDegrees=45
  twistMotion=Free
 }
 {
  bone1Name=thigh_r
  bone2Name=pelvis
  swing1LimitDegrees=45
  swing1Motion=Free
  swing2LimitDegrees=45
  swing2Motion=Free
  twistLimitDegrees=45
  twistMotion=Free
 }
 {bone1Name=extra_cloth_l, bone2Name=extra_cloth_MSTR_l, swing1LimitDegrees=170, swing2LimitDegrees=10, twistLimitDegrees=0}
 {bone1Name=extra_cloth_r, bone2Name=extra_cloth_MSTR_r, swing1LimitDegrees=170, swing2LimitDegrees=3, twistLimitDegrees=0}
 {
  bone1Name=extra_cloth_MSTR_r
  bone2Name=pelvis
  swing1LimitDegrees=45
  swing1Motion=Free
  swing2LimitDegrees=45
  swing2Motion=Free
  twistLimitDegrees=45
  twistMotion=Free
 }
 {
  bone1Name=extra_cloth_MSTR_l
  bone2Name=pelvis
  swing1LimitDegrees=45
  swing1Motion=Free
  swing2LimitDegrees=45
  swing2Motion=Free
  twistLimitDegrees=45
  twistMotion=Free
 }
]
