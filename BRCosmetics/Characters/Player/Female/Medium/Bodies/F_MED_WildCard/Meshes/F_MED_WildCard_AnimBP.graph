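# Post-process style Anim Blueprint for the female WildCard cosmetic body, adding corrective bone motion and pants physics over the attached parent's copied pose.
# Upper-arm, rear-thigh and cloth bone offsets blend by limb swing angles measured each update.
# Left rear-thigh blend is interpolated, easing to a fixed value while crouch-moving.
# AnimDynamics springs then simulate six pants bones; RigidBody and front-thigh nodes are unconnected.
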AnimBlueprint /BRCosmetics/Characters/Player/Female/Medium/Bodies/F_MED_WildCard/Meshes/F_MED_WildCard_AnimBP
target_skeleton=&/F_MED_WildCard_Skeleton
graphs [
 AnimGraph {nodes [
   "Transform (Modify) Bone - Bone: extra_front_thigh_l"(Alpha=1, Translation="0,0,0", Rotation="-0.238,40.089,-0.063", Scale="1.273,6.533,1.000") {
    rotationMode=BMM_Additive
    rotationSpace=BCS_ParentBoneSpace
    scaleMode=BMM_Replace
    scaleSpace=BCS_ParentBoneSpace
    translationMode=BMM_Additive
    translationSpace=BCS_ParentBoneSpace
    alphaBoolBlend {blendOption=Linear}
    boneToModify {boneName=extra_front_thigh_l}
    rotation=<rot -0.238 40.0889 -0.063>
    scale=<1.273 6.533>
   }
   "Transform (Modify) Bone - Bone: extra_front_thigh_r"(Alpha=1, Translation="0.000,3.951,0.828", Rotation="1.475,44.805,3.659", Scale="2.141,5.363,0.496") {
    rotationMode=BMM_Replace
    rotationSpace=BCS_ParentBoneSpace
    scaleMode=BMM_Replace
    scaleSpace=BCS_ParentBoneSpace
    translationMode=BMM_Additive
    alphaBoolBlend {blendOption=Linear}
    boneToModify {boneName=extra_front_thigh_r}
    rotation=<rot 1.475 44.805 3.659>
    scale=<2.141 5.363 0.496>
    translation {y=3.951, z=0.828}
   }
   RigidBody() {
    bOverrideWorldGravity=true
    overridePhysicsAsset=&/F_MED_WildCard_Physics
    alphaBoolBlend {blendOption=Linear}
    componentAppliedLinearAccClamp=<500 500 500>
    overrideWorldGravity {z=-980}
   }
   n1 = "Transform (Modify) Bone - Bone: extra_upperarm_l"(
    ComponentPose=Local To Component(LocalPose=Copy Pose From Mesh() {bUseAttachedParent=true})
    Alpha=Get upperarm_l_alpha() -> (upperarm_l_alpha:double)
    Translation="0,0,0"
    Rotation="1.474,15.102,-13.263"
    Scale="1,1,1"
   ) {
    rotationMode=BMM_Additive
    scaleMode=BMM_Replace
    translationMode=BMM_Additive
    translationSpace=BCS_ParentBoneSpace
    alphaBoolBlend {blendOption=Linear}
    boneToModify {boneName=extra_upperarm_l}
    rotation=<rot 1.474 15.102 -13.263>
   }
   n2 = "Transform (Modify) Bone - Bone: extra_upperarm_r"(
    ComponentPose=@n1
    Alpha=Get upperarm_r_alpha() -> (upperarm_r_alpha:double)
    Translation="7.413,-1.668,4.776"
    Rotation="0,0,0"
    Scale="1.319,1.819,1.319"
   ) {
    rotationMode=BMM_Replace
    rotationSpace=BCS_ParentBoneSpace
    scaleMode=BMM_Additive
    translationMode=BMM_Additive
    translationSpace=BCS_ParentBoneSpace
    alphaBoolBlend {blendOption=Linear}
    boneToModify {boneName=extra_upperarm_r}
    scale=<1.319 1.819 1.319>
    translation=<7.413 -1.668 4.776>
   }
   n3 = "Transform (Modify) Bone - Bone: extra_back_thigh_l"(
    ComponentPose=@n2
    Alpha=Get thigh_l_alpha_back() -> (thigh_l_alpha_back:double)
    Translation="15.751,-20.505,0.212"
    Rotation="41.317,20.109,-76.623"
    Scale="1,1,1"
   ) {
    rotationMode=BMM_Replace
    rotationSpace=BCS_ParentBoneSpace
    scaleMode=BMM_Replace
    scaleSpace=BCS_ParentBoneSpace
    translationMode=BMM_Replace
    translationSpace=BCS_ParentBoneSpace
    alphaBoolBlend {blendOption=Linear}
    boneToModify {boneName=extra_back_thigh_l}
    rotation=<rot 41.317 20.1089 -76.623>
    translation=<15.751 -20.505 0.212>
   }
   n4 = "Transform (Modify) Bone - Bone: extra_back_thigh_r"(
    ComponentPose=@n3
    Alpha=Get thigh_r_alpha_back() -> (thigh_r_alpha_back:double)
    Translation="-10,10,-0.12"
    Rotation="0,60,0"
    Scale="1,1,1"
   ) {
    rotationMode=BMM_Replace
    rotationSpace=BCS_ParentBoneSpace
    scaleMode=BMM_Replace
    scaleSpace=BCS_ParentBoneSpace
    translationMode=BMM_Replace
    translationSpace=BCS_ParentBoneSpace
    alphaBoolBlend {blendOption=Linear}
    boneToModify {boneName=extra_back_thigh_r}
    rotation {yaw=60}
    translation=<-10 10 -0.12>
   }
   n5 = "Transform (Modify) Bone - Bone: extra_cloth_l"(
    ComponentPose=@n4
    Alpha=Get cloth_l_alpha() -> (cloth_l_alpha:double)
    Translation="-0.282,-0.257,2.157"
    Rotation="30.202,83.563,-1.236"
    Scale="1,1,1"
   ) {
    rotationMode=BMM_Replace
    rotationSpace=BCS_BoneSpace
    scaleMode=BMM_Replace
    translationMode=BMM_Additive
    alphaBoolBlend {blendOption=Linear}
    boneToModify {boneName=extra_cloth_l}
    rotation=<rot 30.202 83.563 -1.236>
    translation=<-0.282 -0.257 2.157>
   }
   n6 = "Transform (Modify) Bone - Bone: extra_cloth_r"(
    ComponentPose=@n5
    Alpha=Get cloth_r_alpha() -> (cloth_r_alpha:double)
    Translation="-6.827,-4.986,1.308"
    Rotation="22.418,103.488,-30.190"
    Scale="1,1,1"
   ) {
    rotationMode=BMM_Replace
    rotationSpace=BCS_BoneSpace
    scaleMode=BMM_Additive
    translationMode=BMM_Additive
    alphaBoolBlend {blendOption=Linear}
    boneToModify {boneName=extra_cloth_r}
    rotation=<rot 22.418 103.4881 -30.19>
    translation=<-6.827 -4.986 1.308>
   }
   n7 = Anim Dynamics - Bone: dyn_pants_l_inner(ComponentPose=@n6, Alpha=1, ExternalForce="0,0,0") {
    angularDampingOverride=0.999
    angularSpringConstant=5000
    bAngularSpring=true
    bGravityOverrideInSimSpace=true
    bLinearSpring=true
    bOverrideAngularDamping=true
    gravityScale=0.75
    linearSpringConstant=5000
    simulationSpace=RootRelative
    alphaBoolBlend {blendOption=Linear}
    alphaScaleBiasClamp {scale=0.2}
    boundBone {boneName=dyn_pants_l_inner}
    componentAppliedLinearAccClamp=<6000 6000 6000>
    componentLinearAccScale=<0.25 0.25 0.6>
    componentLinearVelScale=<0.25 0.25 0.6>
    physicsBodyDefinitions [{
      sphereCollisionRadius=0
      boundBone {boneName=dyn_pants_l_inner}
      boxExtents=<8 25 25>
      constraintSetup {angularLimitsMax=<60 40 40>, angularLimitsMin=<-60 -40 -75>, angularTarget=<1 0 0>}
      localJointOffset=<5 0 0>
     }]
    relativeSpaceBone {boneName=root}
   }
   n8 = Anim Dynamics - Bone: dyn_pants_l_outer(ComponentPose=@n7, Alpha=1, ExternalForce="0,0,0") {
    angularDampingOverride=0.999
    angularSpringConstant=5000
    bAngularSpring=true
    bGravityOverrideInSimSpace=true
    bLinearSpring=true
    bOverrideAngularDamping=true
    gravityScale=0.75
    linearSpringConstant=5000
    simulationSpace=RootRelative
    alphaBoolBlend {blendOption=Linear}
    alphaScaleBiasClamp {scale=0.2}
    boundBone {boneName=dyn_pants_l_outer}
    componentAppliedLinearAccClamp=<6000 6000 6000>
    componentLinearAccScale=<0.25 0.25 0.6>
    componentLinearVelScale=<0.25 0.25 0.6>
    physicsBodyDefinitions [{
      sphereCollisionRadius=0
      boundBone {boneName=dyn_pants_l_outer}
      boxExtents=<8 25 25>
      constraintSetup {angularLimitsMax=<60 40 40>, angularLimitsMin=<-60 -40 -75>, angularTarget=<1 0 0>}
      localJointOffset=<5 0 0>
     }]
    relativeSpaceBone {boneName=root}
   }
   n9 = Anim Dynamics - Bone: dyn_pants_l_lower(ComponentPose=@n8, Alpha=1, ExternalForce="0,0,0") {
    angularDampingOverride=0.999
    angularSpringConstant=5000
    bAngularSpring=true
    bGravityOverrideInSimSpace=true
    bLinearSpring=true
    bOverrideAngularDamping=true
    gravityScale=0.75
    linearSpringConstant=5000
    simulationSpace=RootRelative
    alphaBoolBlend {blendOption=Linear}
    alphaScaleBiasClamp {scale=0.2}
    boundBone {boneName=dyn_pants_l_lower}
    componentAppliedLinearAccClamp=<6000 6000 6000>
    componentLinearAccScale=<0.25 0.25 0.6>
    componentLinearVelScale=<0.25 0.25 0.6>
    physicsBodyDefinitions [{
      sphereCollisionRadius=0
      boundBone {boneName=dyn_pants_l_lower}
      boxExtents=<8 25 25>
      constraintSetup {angularLimitsMax=<60 40 40>, angularLimitsMin=<-60 -40 -75>, angularTarget=<1 0 0>}
      localJointOffset=<5 0 0>
     }]
    relativeSpaceBone {boneName=root}
   }
   n10 = Anim Dynamics - Bone: dyn_pants_r_inner(ComponentPose=@n9, Alpha=1, ExternalForce="0,0,0") {
    angularDampingOverride=0.999
    angularSpringConstant=5000
    bAngularSpring=true
    bGravityOverrideInSimSpace=true
    bLinearSpring=true
    bOverrideAngularDamping=true
    gravityScale=0.75
    linearSpringConstant=5000
    simulationSpace=RootRelative
    alphaBoolBlend {blendOption=Linear}
    alphaScaleBiasClamp {scale=0.2}
    boundBone {boneName=dyn_pants_r_inner}
    componentAppliedLinearAccClamp=<6000 6000 6000>
    componentLinearAccScale=<0.25 0.25 0.6>
    componentLinearVelScale=<0.25 0.25 0.6>
    physicsBodyDefinitions [{
      sphereCollisionRadius=0
      boundBone {boneName=dyn_pants_r_inner}
      boxExtents=<8 25 25>
      constraintSetup {angularLimitsMax=<60 40 40>, angularLimitsMin=<-60 -40 -75>, angularTarget=<1 0 0>}
      localJointOffset=<5 0 0>
     }]
    relativeSpaceBone {boneName=root}
   }
   n11 = Anim Dynamics - Bone: dyn_pants_r_outer(ComponentPose=@n10, Alpha=1, ExternalForce="0,0,0") {
    angularDampingOverride=0.999
    angularSpringConstant=5000
    bAngularSpring=true
    bGravityOverrideInSimSpace=true
    bLinearSpring=true
    bOverrideAngularDamping=true
    gravityScale=0.75
    linearSpringConstant=5000
    simulationSpace=RootRelative
    alphaBoolBlend {blendOption=Linear}
    alphaScaleBiasClamp {scale=0.2}
    boundBone {boneName=dyn_pants_r_outer}
    componentAppliedLinearAccClamp=<6000 6000 6000>
    componentLinearAccScale=<0.25 0.25 0.6>
    componentLinearVelScale=<0.25 0.25 0.6>
    physicsBodyDefinitions [{
      sphereCollisionRadius=0
      boundBone {boneName=dyn_pants_r_outer}
      boxExtents=<8 25 25>
      constraintSetup {angularLimitsMax=<60 40 40>, angularLimitsMin=<-60 -40 -75>, angularTarget=<1 0 0>}
      localJointOffset=<5 0 0>
     }]
    relativeSpaceBone {boneName=root}
   }
   n12 = Anim Dynamics - Bone: dyn_pants_r_lower(ComponentPose=@n11, Alpha=1, ExternalForce="0,0,0") {
    angularDampingOverride=0.999
    angularSpringConstant=5000
    bAngularSpring=true
    bGravityOverrideInSimSpace=true
    bLinearSpring=true
    bOverrideAngularDamping=true
    gravityScale=0.75
    linearSpringConstant=5000
    simulationSpace=RootRelative
    alphaBoolBlend {blendOption=Linear}
    alphaScaleBiasClamp {scale=0.2}
    boundBone {boneName=dyn_pants_r_lower}
    componentAppliedLinearAccClamp=<6000 6000 6000>
    componentLinearAccScale=<0.25 0.25 0.6>
    componentLinearVelScale=<0.25 0.25 0.6>
    physicsBodyDefinitions [{
      sphereCollisionRadius=0
      boundBone {boneName=dyn_pants_r_lower}
      boxExtents=<8 25 25>
      constraintSetup {angularLimitsMax=<60 40 40>, angularLimitsMin=<-60 -40 -75>, angularTarget=<1 0 0>}
      localJointOffset=<5 0 0>
     }]
    relativeSpaceBone {boneName=root}
   }
   Output Pose(Result=Component To Local(ComponentPose=@n12)) {layerGroup=None}
  ]}
 EventGraph {nodes [
   n1 = Event BlueprintUpdateAnimation()
   > n2 = Branch(Condition=AND Boolean(A=Get bIsMoving2D(), B=Get bIsCrouching()))
   > n3 = Set thigh_l_alpha_back(thigh_l_alpha_back:double=FInterpTo(Current=Get thigh_l_alpha_back(), Target=0.2, DeltaTime=@n1.DeltaTimeX, InterpSpeed=10)) -> (Output_Get:double)
   n4 = GetOwnerSkeletalMeshComponent()
   Set thigh_r_alpha_back(
    execute=[@n6.then, @n3.then]
    thigh_r_alpha_back:double=MapRangeClamped(Value=GetTwistAndSwingAngleOfDeltaRotationFromRefPose(self=@n4, BoneName=thigh_r).OutSwingAngle, InRangeA=169, InRangeB=69)
   ) -> (Output_Get:double)
   > Set cloth_l_alpha(cloth_l_alpha:double=MapRangeClamped(Value=GetTwistAndSwingAngleOfDeltaRotationFromRefPose(self=@n4, BoneName=thigh_l).OutSwingAngle, InRangeA=20, InRangeB=110)) -> (Output_Get:double)
   > Set cloth_r_alpha(cloth_r_alpha:double=MapRangeClamped(Value=GetTwistAndSwingAngleOfDeltaRotationFromRefPose(self=@n4, BoneName=thigh_r).OutSwingAngle, InRangeA=149, InRangeB=69)) -> (Output_Get:double)
   > Set upperarm_l_alpha(upperarm_l_alpha:double=MapRangeClamped(Value=GetTwistAndSwingAngleOfDeltaRotationFromRefPose(self=@n4, BoneName=upperarm_l).OutSwingAngle, InRangeA=56, InRangeB=90)) -> (Output_Get:double)
   > Set upperarm_r_alpha(upperarm_r_alpha:double=MapRangeClamped(Value=GetTwistAndSwingAngleOfDeltaRotationFromRefPose(self=@n4, BoneName=upperarm_r).OutSwingAngle, InRangeA=66, InRangeB=130)) -> (Output_Get:double)
   n5 = GetTwistAndSwingAngleOfDeltaRotationFromRefPose(self=@n4, BoneName=thigh_l)
   n6 = Set thigh_l_alpha_back(
    execute=@n2.else
    thigh_l_alpha_back=FInterpTo(
     Current=Get thigh_l_alpha_back()
     Target=MapRangeClamped(Value=@n5.OutSwingAngle, InRangeA=24, InRangeB=110)
     DeltaTime=@n1.DeltaTimeX
     InterpSpeed=10
    )
   )
  ]}
]
variables [
 cloth_l_alpha {type=number}
 cloth_r_alpha {type=number}
 thigh_l_alpha_back {type=number}
 thigh_r_alpha_back {type=number, default=1}
 upperarm_l_alpha {type=number}
 upperarm_r_alpha {type=number}
]
