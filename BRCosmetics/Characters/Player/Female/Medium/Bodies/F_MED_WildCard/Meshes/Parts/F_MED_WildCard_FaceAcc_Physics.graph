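# Physics asset for the WildCard female medium character's face accessory part, setting up physical motion for a dynamic ponytail.
# Capsules on the upper spine, clavicles, neck and head give the ponytail body shapes to collide against.
# The ponytail is a base-plus-seven-bone capsule chain, with an extra angled box on the last bone.
# Chain joints use limited swing and tight twist, while the base-to-head and body joints rotate freely.
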PhysicsAsset /BRCosmetics/Characters/Player/Female/Medium/Bodies/F_MED_WildCard/Meshes/Parts/F_MED_WildCard_FaceAcc_Physics
bodies {
 clavicle_l [{shapeType=Capsule, length=9, radius=6.5, shapeName=clavicle_l_capsule, center=<3.7866 0 -0.8159>, rotation=<rot 90 0 0>}]
 clavicle_r [{shapeType=Capsule, length=9, radius=6.5, shapeName=clavicle_l_capsule, center=<-3.7866 0 0.8159>, rotation=<rot 90 0 180>}]
 dyn_ponytail_01 [{shapeType=Capsule, length=1.25, radius=2.75, shapeName=dyn_ponytail_01_capsule, center=<0 -0.3415 0>, rotation=<rot 0 16.177 90>}]
 dyn_ponytail_02 [{shapeType=Capsule, length=0.25, radius=3.25, shapeName=dyn_ponytail_02_capsule, center=<0 0 -0.3669>}]
 dyn_ponytail_03 [{shapeType=Capsule, length=2.25, radius=2.75, shapeName=dyn_ponytail_03_capsule, center=<0.524 0 -0.2401>}]
 dyn_ponytail_04 [{shapeType=Capsule, length=3.75, radius=3, shapeName=dyn_ponytail_04_capsule, center=<0 0 -0.2966>}]
 dyn_ponytail_05 [{shapeType=Capsule, length=4.25, radius=3.75, shapeName=dyn_ponytail_05_capsule, center=<0 -0.2749 0.1139>}]
 dyn_ponytail_06 [{shapeType=Capsule, length=7.25, radius=4.25, shapeName=dyn_ponytail_06_capsule, center=<-1.6273 0.4945 -0.053>}]
 dyn_ponytail_07 [
  {shapeType=Capsule, length=7, radius=2.25, shapeName=dyn_ponytail_07_capsule, center=<-0.1514 0.3945 0.074>}
  {
   shapeType=Box
   extentX=4
   extentY=1.5
   extentZ=6
   shapeName=dyn_ponytail_07_box
   center=<3.2009 0.6808 -1.7626>
   rotation=<rot 52.8899 0.0002 0>
  }
 ]
 dyn_ponytail_base [{shapeType=Capsule, length=1, radius=2, shapeName=dyn_ponytail_base_capsule}]
 head [{shapeType=Capsule, length=6.5, radius=6.25, shapeName=head_capsule, center=<3.8248 -0.1023 0>, rotation=<rot 90 0 15.496>}]
 neck_01 [{shapeType=Capsule, length=1.5, radius=4.5, shapeName=neck_01_capsule, center=<0.315 0 0>, rotation=<rot 90 0 0>}]
 neck_02 [{shapeType=Capsule, length=2, radius=4, shapeName=neck_02_capsule, center=<2.0289 0 0>, rotation=<rot 90 0 0>}]
 spine_04 [{shapeType=Capsule, length=9.5, radius=8, shapeName=spine_04_capsule, center=<3.1677 -0.6676 0>}]
 spine_05 [{shapeType=Capsule, length=12.25, radius=8.75, shapeName=spine_05_capsule, center=<2.6589 0 0>}]
}
constraints [
 {
  bone1Name=neck_02
  bone2Name=neck_01
  swing1LimitDegrees=45
  swing1Motion=Free
  swing2LimitDegrees=45
  swing2Motion=Free
  twistLimitDegrees=45
  twistMotion=Free
 }
 {
  bone1Name=head
  bone2Name=neck_02
  swing1LimitDegrees=45
  swing1Motion=Free
  swing2LimitDegrees=45
  swing2Motion=Free
  twistLimitDegrees=45
  twistMotion=Free
 }
 {
  bone1Name=dyn_ponytail_base
  bone2Name=head
  swing1LimitDegrees=45
  swing1Motion=Free
  swing2LimitDegrees=45
  swing2Motion=Free
  twistLimitDegrees=45
  twistMotion=Free
 }
 {bone1Name=dyn_ponytail_01, bone2Name=dyn_ponytail_base, swing1LimitDegrees=25, swing2LimitDegrees=25, twistLimitDegrees=5}
 {bone1Name=dyn_ponytail_02, bone2Name=dyn_ponytail_01, swing1LimitDegrees=25, swing2LimitDegrees=25, twistLimitDegrees=5}
 {bone1Name=dyn_ponytail_03, bone2Name=dyn_ponytail_02, swing1LimitDegrees=30, swing2LimitDegrees=30, twistLimitDegrees=5}
 {bone1Name=dyn_ponytail_04, bone2Name=dyn_ponytail_03, swing1LimitDegrees=25, swing2LimitDegrees=25, twistLimitDegrees=5}
 {bone1Name=dyn_ponytail_05, bone2Name=dyn_ponytail_04, swing1LimitDegrees=30, swing2LimitDegrees=30, twistLimitDegrees=5}
 {bone1Name=dyn_ponytail_06, bone2Name=dyn_ponytail_05, swing1LimitDegrees=40, swing2LimitDegrees=25, twistLimitDegrees=5}
 {bone1Name=dyn_ponytail_07, bone2Name=dyn_ponytail_06, swing1LimitDegrees=30, swing2LimitDegrees=10, twistLimitDegrees=5}
 {
  bone1Name=clavicle_l
  bone2Name=spine_05
  swing1LimitDegrees=45
  swing1Motion=Free
  swing2LimitDegrees=45
  swing2Motion=Free
  twistLimitDegrees=45
  twistMotion=Free
 }
 {
  bone1Name=neck_01
  bone2Name=spine_05
  swing1LimitDegrees=45
  swing1Motion=Free
  swing2LimitDegrees=45
  swing2Motion=Free
  twistLimitDegrees=45
  twistMotion=Free
 }
 {
  bone1Name=spine_05
  bone2Name=spine_04
  swing1LimitDegrees=45
  swing1Motion=Free
  swing2LimitDegrees=45
  swing2Motion=Free
  twistLimitDegrees=45
  twistMotion=Free
 }
 {
  bone1Name=clavicle_r
  bone2Name=spine_05
  swing1LimitDegrees=45
  swing1Motion=Free
  swing2LimitDegrees=45
  swing2Motion=Free
  twistLimitDegrees=45
  twistMotion=Free
 }
]
