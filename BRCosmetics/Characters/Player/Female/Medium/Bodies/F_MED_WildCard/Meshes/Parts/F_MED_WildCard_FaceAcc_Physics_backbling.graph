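# Physics asset for the WildCard female character's face accessory, backbling variant, defining collision and joint limits for a dynamic ponytail.
# A ponytail chain runs from the head through a base and seven segments, with tight swing and twist limits below a free-moving base.
# Capsules on upper spine, clavicles, neck and head form the body volume, linked by free angular constraints.
# The upper spine and last three ponytail segments carry second offset shapes, including a box at the tip.
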PhysicsAsset /BRCosmetics/Characters/Player/Female/Medium/Bodies/F_MED_WildCard/Meshes/Parts/F_MED_WildCard_FaceAcc_Physics_backbling
bodies {
 clavicle_l [{shapeType=Capsule, length=9, radius=6.5, shapeName=clavicle_l_capsule, center=<3.7866 0 -0.8159>, rotation=<rot 90 0 0>}]
 clavicle_r [{shapeType=Capsule, length=9, radius=6.5, shapeName=clavicle_l_capsule, center=<-3.7866 0 0.8159>, rotation=<rot 90 0 180>}]
 dyn_ponytail_01 [{shapeType=Capsule, length=3.75, radius=2.75, shapeName=dyn_ponytail_01_capsule, center=<0.008 0.026 -0.0044>, rotation=<rot 0 16.177 90>}]
 dyn_ponytail_02 [{shapeType=Capsule, length=3, radius=3, shapeName=dyn_ponytail_02_capsule, center=<0 0 0.0183>}]
 dyn_ponytail_03 [{shapeType=Capsule, length=4.75, radius=2.5, shapeName=dyn_ponytail_03_capsule, center=<0.0264 0 -0.0315>}]
 dyn_ponytail_04 [{shapeType=Capsule, length=6.25, radius=2.75, shapeName=dyn_ponytail_04_capsule, center=<-0.2325 0.1498 -0.0178>}]
 dyn_ponytail_05 [
  {shapeType=Capsule, length=8, radius=2.5, shapeName=dyn_ponytail_05_capsule, center=<0.019 -0.2658 0.0049>, rotation=<rot 0 -10.0528 0>}
  {shapeType=Capsule, length=10.8465, radius=1.8529, shapeName=dyn_ponytail_05_capsule, center=<4.0659 0.6115 -0.2926>}
 ]
 dyn_ponytail_06 [
  {shapeType=Capsule, length=10, radius=2.5, shapeName=dyn_ponytail_06_capsule, center=<-0.6001 -0.6861 0.0179>, rotation=<rot 0 0.2784 0>}
  {shapeType=Capsule, length=9.5633, radius=1.4536, shapeName=dyn_ponytail_06_capsule, center=<1.1066 0.7078 -0.6632>}
 ]
 dyn_ponytail_07 [
  {shapeType=Capsule, length=7.75, radius=2, shapeName=dyn_ponytail_07_capsule, center=<-0.404 0.3742 -0.0282>}
  {
   shapeType=Box
   extentX=4
   extentY=1.5
   extentZ=6
   shapeName=dyn_ponytail_07_box
   center=<3.2883 0.861 -1.7454>
   rotation=<rot 52.318 11.6149 2.9827>
  }
 ]
 dyn_ponytail_base [{shapeType=Capsule, length=1, radius=2, shapeName=dyn_ponytail_base_capsule}]
 head [{shapeType=Capsule, length=6.5, radius=6.25, shapeName=head_capsule, center=<3.8248 -0.1023 0>, rotation=<rot 90 0 15.496>}]
 neck_01 [{shapeType=Capsule, length=1.5, radius=4.5, shapeName=neck_01_capsule, center=<0.315 0 0>, rotation=<rot 90 0 0>}]
 neck_02 [{shapeType=Capsule, length=2, radius=4, shapeName=neck_02_capsule, center=<2.0289 0 0>, rotation=<rot 90 0 0>}]
 spine_04 [{shapeType=Capsule, length=9.5, radius=8, shapeName=spine_04_capsule, center=<3.1677 -0.6676 0>}]
 spine_05 [
  {shapeType=Capsule, length=27.5, radius=8.75, shapeName=spine_05_capsule, center=<2.6589 0.4558 0>}
  {shapeType=Capsule, length=30.25, radius=9, shapeName=spine_05_capsule, center=<12.7207 5.8143 0>}
 ]
}
constraints [
 {
  bone1Name=neck_02
  bone2Name=neck_01
  swing1LimitDegrees=45
  swing1Motion=Free
  swing2LimitDegrees=45
  swing2Motion=Free
  twistLimitDegrees=45
  twistMotion=Free
 }
 {
  bone1Name=head
  bone2Name=neck_02
  swing1LimitDegrees=45
  swing1Motion=Free
  swing2LimitDegrees=45
  swing2Motion=Free
  twistLimitDegrees=45
  twistMotion=Free
 }
 {
  bone1Name=dyn_ponytail_base
  bone2Name=head
  swing1LimitDegrees=45
  swing1Motion=Free
  swing2LimitDegrees=45
  swing2Motion=Free
  twistLimitDegrees=45
  twistMotion=Free
 }
 {bone1Name=dyn_ponytail_01, bone2Name=dyn_ponytail_base, swing1LimitDegrees=15, swing2LimitDegrees=15, twistLimitDegrees=10}
 {bone1Name=dyn_ponytail_02, bone2Name=dyn_ponytail_01, swing1LimitDegrees=15, swing2LimitDegrees=20, twistLimitDegrees=10}
 {bone1Name=dyn_ponytail_03, bone2Name=dyn_ponytail_02, swing1LimitDegrees=15, swing2LimitDegrees=20, twistLimitDegrees=10}
 {bone1Name=dyn_ponytail_07, bone2Name=dyn_ponytail_06, swing1LimitDegrees=15, swing2LimitDegrees=15, twistLimitDegrees=10}
 {
  bone1Name=clavicle_l
  bone2Name=spine_05
  swing1LimitDegrees=45
  swing1Motion=Free
  swing2LimitDegrees=45
  swing2Motion=Free
  twistLimitDegrees=45
  twistMotion=Free
 }
 {
  bone1Name=neck_01
  bone2Name=spine_05
  swing1LimitDegrees=45
  swing1Motion=Free
  swing2LimitDegrees=45
  swing2Motion=Free
  twistLimitDegrees=45
  twistMotion=Free
 }
 {
  bone1Name=spine_05
  bone2Name=spine_04
  swing1LimitDegrees=45
  swing1Motion=Free
  swing2LimitDegrees=45
  swing2Motion=Free
  twistLimitDegrees=45
  twistMotion=Free
 }
 {
  bone1Name=clavicle_r
  bone2Name=spine_05
  swing1LimitDegrees=45
  swing1Motion=Free
  swing2LimitDegrees=45
  swing2Motion=Free
  twistLimitDegrees=45
  twistMotion=Free
 }
 {bone1Name=dyn_ponytail_04, bone2Name=dyn_ponytail_03, swing1LimitDegrees=15, swing2LimitDegrees=20, twistLimitDegrees=10}
 {bone1Name=dyn_ponytail_05, bone2Name=dyn_ponytail_04, swing1LimitDegrees=20, swing2LimitDegrees=15, twistLimitDegrees=10}
 {bone1Name=dyn_ponytail_06, bone2Name=dyn_ponytail_05, swing1LimitDegrees=25, swing2LimitDegrees=15, twistLimitDegrees=10}
]
