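# Post-copy animation Blueprint for the WildCard female FaceAcc part, adding physics and pose fixes to its ponytail bones.
# It copies the attached parent's pose, then blends by bool between two RigidBody chains: a head-space backbling physics asset for non-cape backpacks, and the default asset otherwise.
# Additive bone offsets adjust the ponytail for skydiving, floating and capes.
# Each update detects backbling type and damps the default chain's motion scales while crouching.
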AnimBlueprint /BRCosmetics/Characters/Player/Female/Medium/Bodies/F_MED_WildCard/Meshes/Parts/F_MED_WildCard_FaceAcc_AnimBP
target_skeleton=&/F_MED_WildCard_FaceAcc_Skeleton
graphs [
 AnimGraph {nodes [
   n1 = RigidBody!(
    ComponentPose=Local To Component(LocalPose=Copy Pose From Mesh() {bUseAttachedParent=true})
    bAlphaBoolEnabled:bool=NOT Boolean(A=Get bIsSkyDiving())
    ExternalForce
   ) -> (Pose) {
    alphaInputType=Bool
    overridePhysicsAsset=&/F_MED_WildCard_FaceAcc_Physics_backbling
    simulationSpace=BaseBoneSpace
    alphaBoolBlend {blendOption=Linear}
    baseBoneRef {boneName=head}
    componentLinearAccScale=<0.1 0.1 0.1>
    componentLinearVelScale=<0.1 0.1 0.1>
    simSpaceSettings {maxAngularAcceleration=100, maxAngularVelocity=100, maxLinearAcceleration=100, maxLinearVelocity=100}
   }
   n2 = "Transform (Modify) Bone - Bone: dyn_ponytail_05"!(ComponentPose=@n1, bAlphaBoolEnabled:bool=True, Rotation="0,1,0") -> (Pose) {
    alphaInputType=Bool
    rotationMode=BMM_Additive
    rotationSpace=BCS_BoneSpace
    scaleMode=BMM_Additive
    scaleSpace=BCS_BoneSpace
    translationMode=BMM_Additive
    translationSpace=BCS_BoneSpace
    alphaBoolBlend {blendOption=Linear}
    boneToModify {boneName=dyn_ponytail_05}
    rotation {yaw=1}
    scale {y=0.65}
    translation {y=-0.5}
   }
   n3 = Get bIsSkydiveFloating()
   n4 = "Transform (Modify) Bone - Bone: dyn_ponytail_04"!(ComponentPose=@n2, bAlphaBoolEnabled:bool=@n3, Translation="0,0,0", Rotation="0.000,-59.219,0.000", Scale="1,1,1") -> (Pose) {
    alphaInputType=Bool
    rotationMode=BMM_Additive
    rotationSpace=BCS_BoneSpace
    translationMode=BMM_Additive
    translationSpace=BCS_BoneSpace
    alphaBoolBlend {blendOption=Linear}
    boneToModify {boneName=dyn_ponytail_04}
    rotation=<rot 0 -59.2189 0>
   }
   n5 = "Transform (Modify) Bone - Bone: dyn_ponytail_05"!(ComponentPose=@n4, bAlphaBoolEnabled=@n3, Translation:Vector="0,0,0", Rotation="0.000,27.235,0.000", Scale:Vector="1,1,1") -> (Pose) {
    alphaInputType=Bool
    rotationMode=BMM_Additive
    rotationSpace=BCS_BoneSpace
    translationMode=BMM_Additive
    translationSpace=BCS_BoneSpace
    alphaBoolBlend {blendOption=Linear}
    boneToModify {boneName=dyn_ponytail_05}
    rotation=<rot 0 27.235 0>
   }
   n6 = RigidBody!(
    ComponentPose=Local To Component(LocalPose=Copy Pose From Mesh() {bUseAttachedParent=true})
    bAlphaBoolEnabled
    OverrideWorldGravity:Vector=Get GravityOverride()
    ExternalForce=Get External_Force() -> (External_Force:Vector)
    ComponentLinearAccScale:Vector=Get Linear_Acc_Scale() -> (Linear_Acc_Scale:Vector)
    ComponentLinearVelScale:Vector=Get Linear_Vel_Scale() -> (Linear_Vel_Scale:Vector)
   ) -> (Pose) {
    alphaInputType=Bool
    bClampLinearTranslationLimitToRefPose=true
    bOverrideWorldGravity=true
    overridePhysicsAsset=&/F_MED_WildCard_FaceAcc_Physics
    alphaBoolBlend {blendOption=Linear}
    baseBoneRef {boneName=spine_04}
    componentAppliedLinearAccClamp=<3000 3000 3000>
    componentLinearAccScale=<0.5 0.5 0.5>
    componentLinearVelScale=<0.5 0.5 0.5>
    overrideWorldGravity {z=-490}
   }
   n7 = "Transform (Modify) Bone - Bone: dyn_ponytail_base"!(ComponentPose=@n6, bAlphaBoolEnabled:bool=Get bIsSkyDiving(), Translation="0,0,0", Rotation="0.000,-65.684,-0.000", Scale="1,1,1") -> (Pose) {
    alphaInputType=Bool
    rotationMode=BMM_Additive
    rotationSpace=BCS_BoneSpace
    translationMode=BMM_Additive
    translationSpace=BCS_BoneSpace
    alphaBoolBlend {blendOption=Linear}
    boneToModify {boneName=dyn_ponytail_base}
    rotation=<rot 0 -65.6838 0>
   }
   n8 = Get cape() -> (cape:bool)
   n9 = "Transform (Modify) Bone - Bone: dyn_ponytail_05"(ComponentPose=@n7, bAlphaBoolEnabled=@n8, Translation="0,0,0", Rotation="-0.002,-20.097,-0.029", Scale="1,1,1") {
    alphaInputType=Bool
    rotationMode=BMM_Additive
    rotationSpace=BCS_BoneSpace
    translationMode=BMM_Additive
    translationSpace=BCS_BoneSpace
    alphaBoolBlend {blendOption=Linear}
    boneToModify {boneName=dyn_ponytail_05}
    rotation=<rot -0.002 -20.0969 -0.029>
   }
   n10 = "Transform (Modify) Bone - Bone: dyn_ponytail_06"!(ComponentPose=@n9, bAlphaBoolEnabled:bool=@n8, Translation="0,0,0", Rotation="-0.015,-18.852,0.046", Scale="1,1,1") -> (Pose) {
    alphaInputType=Bool
    rotationMode=BMM_Additive
    rotationSpace=BCS_BoneSpace
    translationMode=BMM_Additive
    translationSpace=BCS_BoneSpace
    alphaBoolBlend {blendOption=Linear}
    boneToModify {boneName=dyn_ponytail_06}
    rotation=<rot -0.015 -18.852 0.046>
   }
   Output Pose(Result=Blend Poses by bool(
     BlendPose_0=Component To Local(ComponentPose=@n5)
     BlendPose_1=Component To Local(ComponentPose=@n10)
     bActiveValue=Get backpack() -> (backpack:bool)
    ) {blendType=Linear}) {layerGroup=None}
  ]}
 EventGraph {nodes [
   Event BlueprintUpdateAnimation()
   n1 = CheckForBackbling()
   > Set cape(cape:bool=@n1.bIsBackblingACape) -> (Output_Get:bool)
   > Set backpack(backpack:bool=AND Boolean(A=@n1, B=NOT Boolean(A=@n1.bIsBackblingACape))) -> (Output_Get:bool)
   > n2 = Branch(Condition=Get bIsCrouching())
   > Set External_Force(External_Force:Vector="0, 0, 0") -> (Output_Get:Vector)
   > Set GravityOverride(GravityOverride="0,0,-980")
   > Set Linear_Acc_Scale(Linear_Acc_Scale:Vector="0.1,0.1,0.1") -> (Output_Get:Vector)
   > Set Linear_Vel_Scale(Linear_Vel_Scale:Vector="0.1,0.1,0.1") -> (Output_Get:Vector)
   Set External_Force(execute=@n2.else, External_Force="0, 0, 0")
   > Set GravityOverride(GravityOverride="0,0,-980")
   > Set Linear_Acc_Scale(Linear_Acc_Scale="0.3,0.3,0.35")
   > Set Linear_Vel_Scale(Linear_Vel_Scale="0.3,0.3,0.5")
  ]}
]
variables [
 backpack {type=boolean}
 cape {type=boolean}
 External_Force {type=Vector}
 GravityOverride {type=Vector}
 Linear_Acc_Scale {type=Vector}
 Linear_Vel_Scale {type=Vector}
]
